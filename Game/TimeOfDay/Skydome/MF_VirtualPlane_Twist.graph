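# Material function for the time-of-day skydome that twists a position around the world Z axis to swirl virtual-plane lookups.
# The input position's XY, divided by a tiling parameter, samples the swirl texture, and a selectable channel scales the RotationAngle input.
# RotateAboutAxis uses that per-position angle around the origin, and its offset is added back to output the twisted position.
# A texture-coordinate SwirlMaskTiling multiply is left unconnected.
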MaterialFunction /Game/TimeOfDay/Skydome/MF_VirtualPlane_Twist
graph {nodes [
  n1 = FunctionInput(Preview=WorldPosition().XYZ) {inputName=Worldposition, sortPriority=-1}
  n2 = Constant3Vector() {constant=<rgba 0 0 1 0>}
  n3 = FunctionInput() {bUsePreviewValueAsDefault=true, inputName=RotationAngle, inputType=FunctionInput_Scalar, previewValue=<1 0 0 1>}
  n4 = ComponentMask(None=Divide(A=@n1, B=ScalarParameter() {defaultValue=5535037, group=Twist, parameterName=Twist-MaskTiling})) {g=true, r=true}
  n5 = FunctionInput() {inputName=SwirlTexture, inputType=FunctionInput_Texture2D}
  n6 = Multiply(
   A=@n3
   B=StaticComponentMaskParameter(None=TextureSample(UVs=@n4, Tex=@n5) {samplerSource=SSM_Wrap_WorldGroupSettings, texture=/Game/TimeOfDay/LightFunction/T_CloudNoise_Artemis}.RGBA) {defaultA=true, group=Twist, parameterName=Twist-TextureChannel}
  )
  n7 = Constant3Vector()
  FunctionOutput(None=Add(A=@n1, B=RotateAboutAxis(NormalizedRotationAxis=@n2, RotationAngle=@n6, PivotPoint=@n7, Position=@n1)))
  Multiply(A=TextureCoordinate(), B=ScalarParameter() {defaultValue=0.3, group=Twist, parameterName=SwirlMaskTiling})
 ]}
