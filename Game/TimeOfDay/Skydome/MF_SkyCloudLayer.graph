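# Skydome material function that renders one panning cloud layer lit by the Sky Atmosphere.
# It projects onto a virtual plane at a given altitude, optionally twisted, and samples a scrolling noise channel for opacity.
# Color sums skylight-scaled scattered luminance and sun illuminance, with view-dependent rim light on thin areas at Epic quality only.
# Aerial perspective alpha blends that color with sky view luminance; position, illuminance and aerial perspective are also output.
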MaterialFunction /Game/TimeOfDay/Skydome/MF_SkyCloudLayer
graph {nodes [
  n1 = FunctionInput() {bUsePreviewValueAsDefault=true, inputName=SkylightBrightness, inputType=FunctionInput_Scalar, previewValue=<1 0 0 1>}
  n2 = FunctionInput() {bUsePreviewValueAsDefault=true, inputName=BaseColor, previewValue=<1 1 1 1>}
  n3 = Multiply(A=Multiply(A=SkyAtmosphereDistantLightScatteredLuminance(), B=@n1), B=@n2)
  n4 = Constant2Vector()
  n5 = FunctionInput() {bUsePreviewValueAsDefault=true, inputName="Altitude(cm)", inputType=FunctionInput_Scalar, previewValue=<6 0 0 1>}
  n6 = FunctionOutput(None=MaterialFunctionCall(Plane Center=AppendVector(A=@n4, B=@n5)) {materialFunction=&VirtualPlaneCoordinates}.World Position) {outputName=WorldPosition}
  n7 = StaticSwitch(
   True=MaterialFunctionCall!(
    Worldposition=@n6
    SwirlTexture=NamedRerouteUsage()
    RotationAngle=ScalarParameter() {defaultValue=0.15, group=Twist, parameterName=TwistAngle}
   ) {materialFunction=&/MF_VirtualPlane_Twist}.Result
   False=@n6
   Value=FunctionInput(Preview=StaticBool()) {bUsePreviewValueAsDefault=true, inputName=TwistClouds, inputType=FunctionInput_StaticBool}
  )
  n8 = FunctionInput(Preview=SkyAtmosphereLightIlluminance!(World Position=@n7)) {bUsePreviewValueAsDefault=true, inputName=LightIlluminance}
  n9 = Multiply(A=Constant() {desc="PRETTY SIMPLE LIGHTING :)", r=1}, B=@n2)
  n10 = FunctionInput() {bUsePreviewValueAsDefault=true, inputName=Speed, inputType=FunctionInput_Scalar, previewValue=<8000 0 0 1>}
  n11 = Multiply(A=Multiply(A=Time(), B=@n10), B=FunctionInput() {bUsePreviewValueAsDefault=true, inputName=PanDirection, previewValue=<-1 1 0 1>})
  n12 = FunctionInput() {bUsePreviewValueAsDefault=true, inputName="Size(cm)", inputType=FunctionInput_Scalar, previewValue=<29.33 0 0 1>}
  n13 = ComponentMask(None=Divide(A=Add(A=@n7, B=@n11), B=@n12)) {g=true, r=true}
  n14 = NamedRerouteDeclaration(None=FunctionInput(Preview=TextureObject() {texture=/Game/TimeOfDay/LightFunction/T_CloudNoise_Artemis}) {bUsePreviewValueAsDefault=true, inputName=CloudAlpha, inputType=FunctionInput_Texture2D, sortPriority=-5}) {name=CloudTexture, nodeColor=<rgba 1 0 0.8317 1>}
  n15 = Saturate(None=DotProduct(
    A=TextureSample(UVs=@n13, Tex=@n14) {samplerSource=SSM_Wrap_WorldGroupSettings, samplerType=SAMPLERTYPE_LinearGrayscale}.RGB
    B=FunctionInput() {bUsePreviewValueAsDefault=true, inputName=CloudChannel, previewValue=<0 1 0 1>}
   ))
  n16 = Power(Base=OneMinus(None=@n15), Exp=Constant() {r=3})
  n17 = SkyAtmosphereLightDirection()
  n18 = Power(Base=DotProduct(A=Multiply(A=@n17) {constA=-1, constB=-1}, B=CameraVectorWS()), Exp=Constant() {r=10})
  n19 = FunctionInput() {bUsePreviewValueAsDefault=true, inputName=RimIntensity, inputType=FunctionInput_Scalar, previewValue=<2 0 0 1>}
  n20 = Multiply(A=@n2, B=Multiply(A=Multiply(A=@n16, B=@n18), B=@n19))
  n21 = Multiply(A=Multiply(A=ComponentMask(None=@n8) {b=true, g=true, r=true}) {constB=0.2}, B=QualitySwitch(Default=@n9, Epic=Add(A=@n9, B=@n20)))
  n22 = FunctionOutput(None=FunctionInput(Preview=SkyAtmosphereAerialPerspective!(World Position=@n7)) {bUsePreviewValueAsDefault=true, inputName=AerialPerspective, inputType=FunctionInput_Vector4}) {outputName=AerialPerspective}
  FunctionOutput(None=LinearInterpolate(
    A=SkyAtmosphereViewLuminance!(WorldDirection)
    B=Add(A=@n3, B=@n21)
    Alpha=MaterialFunctionCall(Float4=@n22) {materialFunction=&BreakOutFloat4Components}.A
   )) {outputName=CloudColor}
  n23 = FunctionInput() {bUsePreviewValueAsDefault=true, inputName=CloudOpacity, inputType=FunctionInput_Scalar, previewValue=<0.3 0 0 1>}
  FunctionOutput(None=Saturate(None=Multiply(A=@n23, B=@n15))) {outputName=CloudAlpha, sortPriority=1}
  FunctionOutput(None=@n8) {outputName=LightIlluminance}
 ]}
