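# Unlit sky dome material for Artemis that draws atmosphere, sun and moon disks, clouds, stars and galaxy across the time-of-day cycle.
# Sky atmosphere luminance forms the base, overlaid by two MF_SkyCloudLayer layers and a panning noise cloud mask.
# Triplanar stars use derivative-widened thresholds, with night masks from Fortnite time-of-day parameters.
# Reflection capture switches brighten, darken and tint skylight captures separately, alongside fake dusk and dawn light.
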Material /Game/TimeOfDay/Skydome/M_Skydome_Artemis
graph {
 properties {bIsSky=true, pixelDepthOffsetMode=PDOM_Legacy, shadingModel=MSM_Unlit, usedShadingModels=Unlit}
 nodes [
  Constant() {r=3}
  Constant() {r=2}
  n1 = VectorParameter() {group=CloudMask, parameterName=CloudPanDirection, defaultValue=<rgba -1 1 0 0>}
  n2 = MaterialFunctionCall!(
   CloudAlpha=NamedRerouteUsage()
   SkylightBrightness=ScalarParameter() {defaultValue=1, group=C1, parameterName=C1Skylight}
   "Altitude(cm)"=ScalarParameter() {defaultValue=200000.5, group=C1, parameterName=C1Altitude}
   Speed=ScalarParameter() {defaultValue=4161.4268, group=C1, parameterName=C1Speed}
   "Size(cm)"=ScalarParameter() {defaultValue=1254811.75, group=C1, parameterName=C1Size}
   CloudOpacity=ScalarParameter() {defaultValue=1, group=C1, parameterName=C1Opacity}
   RimIntensity=ScalarParameter() {defaultValue=20, group=C1, parameterName=C1Rim}
   BaseColor=VectorParameter() {group=C1, parameterName=C1Color, defaultValue=<rgba 0.8 0.8 0.8 1>}.RGB
   CloudChannel=VectorParameter() {group=C1, parameterName=C1Channel, defaultValue=<rgba 1 0 0 1>}.RGB
   AerialPerspective
   LightIlluminance
   TwistClouds=StaticBool() {value=true}
   PanDirection=@n1.RGB
  ) {materialFunction=&/MF_SkyCloudLayer}
  n3 = Divide(
   A=ComponentMask(None=@n2.WorldPosition) {g=true, r=true}
   B=ScalarParameter() {defaultValue=1568455.25, group=CloudMask, parameterName=CloudMask-Size}
  )
  n4 = ComponentMask(None=VectorParameter() {group=CloudMask, parameterName=CloudMaskPanningSpeed, defaultValue=<rgba 0.005 0 0 1>}.RGB) {g=true, r=true}
  n5 = ChannelMaskParameter(None=TextureSample(UVs=Panner(Coordinate=@n3, Speed=@n4), Tex=NamedRerouteUsage()) {samplerSource=SSM_Wrap_WorldGroupSettings, texture=/Game/TimeOfDay/LightFunction/T_CloudNoise_Artemis}.RGBA) {group=CloudMask, parameterName=Cloud-Multiply-Mask-TextureChannel}
  Multiply(A=ScalarParameter() {defaultValue=1, group=DEBUG, parameterName=DEBUG-TurnOffClouds}, B=@n5.R)
  n6 = Multiply(A=ComponentMask(None=SkyAtmosphereLightDirection()) {b=true, g=true, r=true}) {constB=-1}
  n7 = CrossProduct(A=@n6, B=Constant3Vector() {constant=<rgba 0 0 -1 0>})
  n8 = Normalize(VectorInput=CrossProduct(A=@n7, B=@n6))
  n9 = NamedRerouteUsage()
  n10 = Normalize(VectorInput=@n7)
  n11 = NamedRerouteUsage()
  n12 = Divide(
   A=AppendVector(A=DotProduct(A=@n8, B=@n9), B=DotProduct(A=@n10, B=@n11))
   B=ScalarParameter() {defaultValue=0.1, group=Sun-Moon, parameterName=MoonSize}
  )
  n13 = TextureSample(UVs=VertexInterpolator(VS=Add(A=@n12) {constB=0.5}).PS) {samplerSource=SSM_Clamp_WorldGroupSettings, samplerType=SAMPLERTYPE_LinearColor, texture=&/T_Moon_Artemis}
  Multiply(
   A=Multiply(A=SkyAtmosphereViewLuminance!(WorldDirection), B=ScalarParameter() {defaultValue=100, group=Sun-Moon, parameterName=RealisticMoonMod})
   B=@n13.R
  )
  NamedRerouteDeclaration(None=@n5.R) {name=CloudMask, nodeColor=<rgba 1 0 0.8317 1>}
  n14 = NamedRerouteUsage()
  n15 = Abs(None=DDX(Value=@n14))
  n16 = Abs(None=DDY(Value=@n14))
  NamedRerouteDeclaration(None=Multiply(A=Add(A=@n15, B=@n16), B=ScalarParameter() {defaultValue=1.5, group=StarsGalaxy, parameterName=StarRadius})) {name=StarRadiusModulation, nodeColor=<rgba 1 0.6797 0 1>}
  NamedRerouteDeclaration(None=@n2.LightIlluminance) {name=ReuseLightIlluminance, nodeColor=<rgba 1 0 0.8317 1>}
  NamedRerouteDeclaration(None=@n2.AerialPerspective) {name=ReuseAerialPerspective, nodeColor=<rgba 1 0 0.8317 1>}
  n17 = Clamp(None=SkyAtmosphereLightDiskLuminance!(DiskAngularDiameterOverride)) {maxDefault=100}
  n18 = LinearInterpolate(A=NamedRerouteUsage(), B=NamedRerouteUsage(), Alpha=NamedRerouteUsage())
  n19 = NamedRerouteUsage()
  n20 = Multiply(A=SkyAtmosphereLightDiskLuminance(), B=ScalarParameter() {defaultValue=0.2, group=Sun-Moon, parameterName=LightDisk-Mod})
  n21 = Multiply(A=NamedRerouteUsage(), B=NamedRerouteUsage())
  n22 = Add(
   A=StaticSwitchParameter(True=LinearInterpolate(A=@n17, B=@n18, Alpha=@n19), False=Add(A=@n20, B=@n21)) {defaultValue=true, group=Sun-Moon, parameterName=UseMoonTexture}
   B=SkyAtmosphereViewLuminance()
  )
  n23 = MaterialFunctionCall(
   CloudAlpha=NamedRerouteUsage()
   SkylightBrightness=ScalarParameter() {defaultValue=1, group=C2, parameterName=c2Skylight}
   "Altitude(cm)"=ScalarParameter() {defaultValue=1500000, group=C2, parameterName=c2Altitude}
   Speed=ScalarParameter() {defaultValue=3061.7397, group=C2, parameterName=c2Speed}
   "Size(cm)"=ScalarParameter() {defaultValue=9500000, group=C2, parameterName=c2Size}
   CloudOpacity=ScalarParameter() {defaultValue=2, group=C2, parameterName=c2Opacity}
   RimIntensity=ScalarParameter() {defaultValue=20, group=C2, parameterName=c2Rim}
   BaseColor=VectorParameter() {group=C2, parameterName=C2Color, defaultValue=<rgba 1 1 1 1>}.RGB
   AerialPerspective=NamedRerouteUsage()
   LightIlluminance=NamedRerouteUsage()
   PanDirection=@n1.RGB
  ) {materialFunction=&MF_SkyCloudLayer}
  n24 = Multiply(A=@n23.CloudAlpha, B=NamedRerouteUsage())
  n25 = NamedRerouteUsage()
  NamedRerouteDeclaration(None=LinearInterpolate(A=LinearInterpolate(A=@n22, B=@n23.CloudColor, Alpha=@n24), B=@n2.CloudColor, Alpha=Multiply(A=@n2.CloudAlpha, B=@n25))) {name=SkyWithClouds, nodeColor=<rgba 1 0 0.8317 1>}
  n26 = CrossProduct(
   A=NamedRerouteUsage()
   B=Normalize(VectorInput=VectorParameter() {group=StarsGalaxy, parameterName=ProjDirection, defaultValue=<rgba -0.0381 -2.3352 0.9775 1>}.RGB)
  )
  n27 = VectorParameter() {group=StarsGalaxy, parameterName=GalaxySize, defaultValue=<rgba 0.2048 0.4048 0 1>}
  n28 = Add(
   A=Multiply(A=ComponentMask(None=@n26) {g=true, r=true}, B=ComponentMask(None=@n27.RGB) {g=true, r=true})
   B=ComponentMask(None=VectorParameter() {group=StarsGalaxy, parameterName=GalaxyOffset, defaultValue=<rgba 0.5 0.5 0 1>}.RGB) {g=true, r=true}
  )
  NamedRerouteDeclaration(None=TextureSample(UVs=VertexInterpolator(VS=@n28).PS) {samplerSource=SSM_Clamp_WorldGroupSettings, texture=&/T_Galaxy_Artemis}.RGB) {name=Galaxy, nodeColor=<rgba 0 1 0.2969 1>}
  n29 = ScalarParameter() {defaultValue=1, group=StarsGalaxy, parameterName=StarThreshold}
  n30 = Add(A=NamedRerouteUsage(), B=ScalarParameter() {defaultValue=0.002, group=StarsGalaxy, parameterName=StarRange})
  n31 = TextureObject() {samplerType=SAMPLERTYPE_LinearGrayscale, texture=&/T_StarField_Artemis}
  n32 = ScalarParameter() {defaultValue=1, group=StarsGalaxy, parameterName=StarTiling}
  n33 = Constant() {r=64}
  n34 = NamedRerouteDeclaration(None=Saturate(None=MaterialFunctionCall!(CameraVector, Texture=@n31, Tiling=@n32, Offset, AxisFadeContrast=@n33) {materialFunction=/Engine/Functions/Engine_MaterialFunctions01/Texturing/TriplanarCameraVector}.Result)) {name=StarGradient, nodeColor=<rgba 0.3125 1 0 1>}
  n35 = NamedRerouteUsage()
  n36 = ScalarParameter() {defaultValue=0.2, group=StarsGalaxy, parameterName=GalaxyBrightness}
  n37 = NamedRerouteUsage()
  n38 = Add(
   A=Multiply(
    A=MaterialFunctionCall!(ThresholdValue=@n29, ThresholdRange=@n30, Alpha=@n34, TwoSidedThreshold) {materialFunction=/Engine/Functions/Engine_MaterialFunctions02/Masking/ThresholdWithRange}.Result
    B=ScalarParameter() {defaultValue=1, group=StarsGalaxy, parameterName=StarBrightness}
   )
   B=Multiply(A=Multiply(A=@n35, B=@n36), B=Power(Base=@n37) {constExponent=0.5})
  )
  n39 = NamedRerouteUsage()
  NamedRerouteDeclaration(None=ReflectionCapturePassSwitch!(Default=Multiply(A=@n38, B=@n39), Reflection=Constant())) {name=Stars, nodeColor=<rgba 1 0 0.7031 1>}
  n40 = ScalarParameter() {defaultValue=0.52, group=NightMask, parameterName=DayCenter}
  n41 = ScalarParameter() {defaultValue=0.77, group=NightMask, parameterName=DayRadius}
  n42 = CollectionParameter() {collection=&FortniteMaterialParameters, parameterName=TimeOfDay_0-1}
  n43 = StaticBool() {value=true}
  n44 = MaterialFunctionCall(
   In=OneMinus(None=MaterialFunctionCall(ThresholdValue=@n40, ThresholdRange=@n41, Alpha=@n42, TwoSidedThreshold=@n43) {materialFunction=&ThresholdWithRange}.Result)
   Contrast=Constant() {r=20}
  ) {materialFunction=&CheapContrast}
  n45 = CollectionParameter() {
   collection=&FortniteMaterialParameters
   parameterName=FortniteActiveTimeOfDay
  }
  n46 = ComponentMask(None=@n45) {a=true, desc=night}
  NamedRerouteDeclaration(None=VertexInterpolator(VS=StaticSwitchParameter(True=@n44.Result, False=@n46) {defaultValue=true, group=NightMask, parameterName=UseType1-NightMask}).PS) {name=NightMask, nodeColor=<rgba 0 0.2734 1 1>}
  NamedRerouteDeclaration(None=Multiply(A=@n13.R, B=VectorParameter() {group=Sun-Moon, parameterName=MoonColor, defaultValue=<rgba 1 1 1 1>}.RGB)) {name=Moon, nodeColor=<rgba 0 1 0.8125 1>}
  n47 = WorldPosition()
  n48 = CameraPositionWS()
  NamedRerouteDeclaration(None=Normalize(VectorInput=Subtract(A=@n47.XYZ, B=@n48))) {name=CameraVectorForVertexShader, nodeColor=<rgba 0 1 0.0859 1>}
  n49 = Multiply(A=@n13.R) {constB=20}
  n50 = Constant() {r=-1}
  n51 = SkyAtmosphereLightDirection()
  n52 = Sign(None=DotProduct(A=Multiply(A=@n50, B=@n51), B=CameraVectorWS()))
  NamedRerouteDeclaration(None=Multiply(A=Saturate(None=@n49), B=Saturate(None=@n52))) {name=MoonMask, nodeColor=<rgba 0 1 0.1094 1>}
  NamedRerouteDeclaration(None=TextureObjectParameter() {group=CloudMask, parameterName=CloudMaskTexture, texture=&T_CloudNoise_Artemis}) {name=CloudMaskTexture, nodeColor=<rgba 1 0 0.8317 1>}
  n53 = NamedRerouteUsage()
  n54 = CollectionParameter() {
   collection=&FortniteMaterialParameters
   parameterName=FortniteActiveTimeOfDay
  }
  n55 = Multiply(
   A=Power(
    Base=ComponentMask(None=@n54) {a=true, desc=night}
    Exp=ScalarParameter() {defaultValue=1, group=FakeBoost, parameterName=NightBoost-TimePower}
   )
   B=ScalarParameter() {defaultValue=1, group=FakeBoost, parameterName=NightBoost}
  )
  ReflectionCapturePassSwitch(Default=@n53, Reflection=Add(A=Multiply(A=@n55, B=@n53), B=@n53))
  n56 = DotProduct(A=SkyAtmosphereLightDirection(), B=Multiply(A=PixelNormalWS()) {constB=-1})
  n57 = ConstantBiasScale(None=@n56)
  n58 = LinearInterpolate(
   A=ScalarParameter() {defaultValue=1.2, group=FakeBoost, parameterName=BrightnessBoostDuringReflectionCapture}
   B=ScalarParameter() {defaultValue=1.5, group=FakeBoost, parameterName=FakeNightLightDuringReflectionCapture}
   Alpha=NamedRerouteUsage()
  ) {constA=1, constB=2}
  n59 = Constant() {r=-1}
  n60 = SkyAtmosphereLightDirection()
  n61 = Clamp(None=DotProduct(A=Multiply(A=@n59, B=@n60), B=NamedRerouteUsage()))
  n62 = LinearInterpolate(A=Constant() {r=0.7}, Alpha=Power(Base=ConstantBiasScale(None=@n61), Exp=Constant() {r=3}))
  n63 = ScalarParameter() {defaultValue=0.05, group=FakeBoost, parameterName=FakeNightLightRange}
  n64 = CollectionParameter() {collection=&FortniteMaterialParameters, parameterName=TimeOfDay_0-1}
  n65 = StaticBool() {value=true}
  n66 = Add(
   A=MaterialFunctionCall(
    ThresholdValue=ScalarParameter() {defaultValue=0.9167, group=FakeBoost, parameterName=FakeNightLight- Dusk Offset}
    ThresholdRange=@n63
    Alpha=@n64
    TwoSidedThreshold=@n65
   ) {materialFunction=&ThresholdWithRange}.Result
   B=MaterialFunctionCall(
    ThresholdValue=ScalarParameter() {defaultValue=0.125, group=FakeBoost, parameterName=FakeNightLight- Dawn Offset}
    ThresholdRange=@n63
    Alpha=@n64
    TwoSidedThreshold=@n65
   ) {materialFunction=&ThresholdWithRange}.Result
  )
  n67 = Multiply(
   A=NamedRerouteDeclaration(None=Power(Base=@n66)) {name=FakeNightLightForCapture, nodeColor=<rgba 1 0 1 1>}
   B=VectorParameter() {group=FakeBoost, parameterName=FakeNightLight, defaultValue=<rgba 0.0657 0.1468 0.2 1>}.RGB
  )
  n68 = Add(A=NamedRerouteUsage(), B=VertexInterpolator(VS=Multiply(A=@n62, B=@n67)).PS)
  n69 = StaticSwitchParameter(True=ReflectionCapturePassSwitch(Default=@n68, Reflection=Multiply(A=@n68, B=@n58)), False=NamedRerouteUsage()) {defaultValue=true, group=FakeBoost, parameterName=UseType1-FakeBoost, sortPriority=-500}
  n70 = Multiply(
   A=@n69
   B=LinearInterpolate(A=ScalarParameter() {defaultValue=0.7, group=ReflectionCaptureOnly, parameterName=DarkenSkyAwayFromSunInSkylight}, Alpha=Power(Base=@n57))
  )
  n71 = Saturate(None=@n56)
  n72 = AtmosphericLightColor()
  n73 = ScalarParameter() {defaultValue=-2, group=ReflectionCaptureOnly, parameterName=AirglowAroundSunDesat}
  n74 = Multiply(
   A=Multiply(A=Power(Base=@n71) {constExponent=10}, B=Desaturation(None=@n72, Fraction=@n73))
   B=ScalarParameter() {defaultValue=0.12, group=ReflectionCaptureOnly, parameterName=AirglowAroundSunInSkylight}
  )
  ReflectionCapturePassSwitch(
   Default=@n69
   Reflection=Desaturation(None=Add(A=@n70, B=@n74), Fraction=ScalarParameter() {group=ReflectionCaptureOnly, parameterName=DesatReflection})
  )
  Saturate(None=Multiply(A=ScalarParameter() {defaultValue=1, group=DEBUG, parameterName=DEBUG-NightMask}))
 ]
}
